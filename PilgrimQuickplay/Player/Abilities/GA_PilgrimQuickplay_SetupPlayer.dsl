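; Gameplay ability that sets up a player's HUD and camera for Pilgrim Quickplay gameplay and reverts both when it ends.
; Activate and end share SetupPlayerController, which toggles tagged HUD element visibility under a named context and sets or clears the Sparks cinematic camera override.
; A separate PlayMontage event, not called here, plays the combat-ready montage for other clients; its note says the local view comes from MusicBattleInstance AnimBP logic and marks it for removal.
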(event Ability|EventActivateAbility
  (|CallParentFunction)
  (CallFunction|SetupPlayerController false))

(event Ability|EventOnEndAbility
  (params (bWasCancelled :type Boolean))
  ;[1 1 1] OnEndAbility
  (|CallParentFunction bWasCancelled)
  (CallFunction|SetupPlayerController true))

(event Custom|PlayMontage
  ;[0.5156 0.3581 0.2742] This effectively plays the montage for OTHER clients, as the server will replicate down this montage. However, this does not dictate what the local player sees -- that is controlled by the animBP logic in MusicBattleInstance. We should tidy it up so everyone is controlled by the animbp and then remove this.
  (Ability|Tasks|PlayMontageAndWait "None" (Variables|Default|GetCombatReadyMontage)))

(fn SetupPlayerController
  (params (OnEndAbility :type Boolean))
  (bind _outfortplayerpawn (CallFunction|GetAvatarasValidFortPlayerPawn)
    (:ValidFortPlayerPawn
      (if OnEndAbility
        (HUD|SetHUDElementVisibilitywithContext (Fortnite|GetFortPlayerControllerfromActor _outfortplayerpawn) (Variables|Default|GetHUDElement%20Tags) false "ForPilgrimQuickplayGameplay" "")
        (Game|Player|SetCinematicOverrideClass (Fortnite|GetFortPlayerControllerfromActor _outfortplayerpawn) 0)
        (else
          (HUD|SetHUDElementVisibilitywithContext (Fortnite|GetFortPlayerControllerfromActor _outfortplayerpawn) (Variables|Default|GetHUDElement%20Tags) true "ForPilgrimQuickplayGameplay" "")
          (Game|Player|SetCinematicOverrideClass (Fortnite|GetFortPlayerControllerfromActor _outfortplayerpawn) "/Script/SparksCameraDirectorRuntime.SparksCinematicCamera"))))))
